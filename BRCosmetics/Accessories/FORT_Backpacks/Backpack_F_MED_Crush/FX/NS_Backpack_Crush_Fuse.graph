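# Niagara system for the Crush backpack's fuse, giving it a glowing tip and trailing smoke at skeletal mesh sockets.
# The fuse emitter pins short-lived particles to the socket, drawn as warm lights, meshes and sprites.
# The smoke emitter spawns ribbon particles with movement, driven by upward gravity, curl noise, drag and socket-relative height.
# Counter-based link order sequences the ribbon, while an eroding material cools orange to grey and scales alpha with owner speed.
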NiagaraSystem /BRCosmetics/Accessories/FORT_Backpacks/Backpack_F_MED_Crush/FX/NS_Backpack_Crush_Fuse
dependencies {
 usedDataInterfaces [&NiagaraDataInterfaceColorCurve, /Script/Niagara.NiagaraDataInterfaceSimpleCounter, &NiagaraDataInterfaceCurve]
 usedDynamicInputs []
 usedModules [&SystemState]
}
emitters {
 FX_Fuse_L_Hand001 {
  emitterUpdateScript [
   &EmitterState(Life Cycle Mode=::NewEnumerator1, Loop Duration=1, UseLoopDelay=false)
   &SpawnBurst_Instantaneous(Spawn Count=3, Use Spawn Probability=false) {enabled=false}
   &SpawnRate(SpawnRate=3, Use Spawn Probability=false)
  ]
  particleSpawnScript [
   /Niagara/Modules/Spawn/Initialization/InitializeParticle(
    Color=<rgba 1 0.5316 0.0955 0.018>
    Lifetime=&RandomRangeFloat(Minimum=2, Maximum=2.5)
    Mesh Scale=&VectorFromFloat(Value=&Multiply_Float(A=0.01, B=&RandomRangeFloat(Minimum=0.3, Maximum=2)))
    Module.Write Color=true
    Module.Write Lifetime=true
    Module.Write Mass=true
    Module.Write Material Random=true
    Module.Write Position=false
    Module.Write Scale=true
    Module.Write SpriteRotation=true
    Module.Write SpriteSize=true
    Position=&SimulationPosition()
    Sprite Rotation=&RandomRangeFloat(Minimum=-1, Maximum=1)
    Sprite Size=&Vector2DFromFloat(Value=7)
   )
   &InitialMeshOrientation(
    Rotation=&RandomRangeVector(Minimum=<0 0 0>, Maximum=<1 1 1>)
    Use Orientation Vector=true
    Use Rotation Vector=true
   )
   &FN_SetSpawnGroup(Number of Spawn Groups=1)
  ]
  particleUpdateScript [
   &ParticleState()
   &SkeletalMeshLocation(
    Bone / Socket Index=&SpawnGroup()
    Inherit Velocity=false
    MaskSpawnGroup=false
    Mesh Sampling Type=::NewEnumerator5
    Sampled Position Offset Coordinate Space:ENiagaraMeshTransforms=::NewEnumerator2
    Skeletal Mesh=$User.SkeletalMesh
    Socket Sampling Mode=::NewEnumerator5
   )
   &SolveForcesAndVelocity(Clamp Velocity=false, Limit Acceleration=false)
   &FN_Light_Attributes(
    Color=&ColorFromCurve(ColorCurve=DataInterface() {propertyValues {
       AlphaCurve {keys [<at 0 0 RCIM_Cubic>,<at 0.279 1>,<at 1 0 RCIM_Cubic>]}
       BlueCurve {keys [<at 0 0 RCIM_Cubic>,<at 0.279 0>,<at 0.5266 0>,<at 1.0077 0.0776>]}
       GreenCurve {keys [<at 0 0.2118 RCIM_Cubic>,<at 0.279 0.3775>,<at 0.5266 0.5161>,<at 1.0077 0.4904>]}
       RedCurve {keys [<at 0 1 RCIM_Cubic>,<at 0.279 1>,<at 0.5266 1>,<at 1.0077 1>]}
      }})
    Create Color Attribute=true
    Create Exponent Attribute=true
    Create Radius Attribute=true
    Radius=20
    Scale Color=1.2
   )
   /SetVariables(Particles.MeshSize:Vector3f=&VectorFromFloat(Value=0.03)) {isSetParameters=true}
  ]
  rendererClasses [
   &NiagaraLightRendererProperties
   &NiagaraMeshRendererProperties
   &NiagaraSpriteRendererProperties
  ]
  renderers [
   {rendererClass=&NiagaraLightRendererProperties, rendererIndex=0}
   {rendererClass=&NiagaraMeshRendererProperties, rendererIndex=1}
   {rendererClass=&NiagaraSpriteRendererProperties, rendererIndex=2}
  ]
 }
 Smoke004 {
  emitterUpdateScript [
   &EmitterState()
   &SpawnRate(SpawnRate=50, Use Spawn Probability=false) {enabled=false}
   &SpawnPerUnit(
    Max Movement Threshold=500
    Module.Use Max Movement Threshold=true
    Module.Use Movement Tolerance=false
    Spawn Per Unit=5
    Use Spawn Probability=false
   )
  ]
  particleSpawnScript [
   &InitializeRibbon(
    Color=<rgba 10 10 10 0.3>
    Lifetime=0.3
    Module.Write Color=true
    Module.Write Lifetime=true
    Module.Write Mass=true
    Module.Write Material Random=true
    Module.Write Position=true
    Module.Write RibbonTwist=false
    Module.Write RibbonWidth=true
    Ribbon Width=&RandomRangeFloat(Minimum=1, Maximum=6)
   )
   /SetVariables(Particles.SpriteSize=&Vector2DFromFloat(Value=1)) {isSetParameters=true}
   &FN_SetSpawnGroup(Number of Spawn Groups=1)
   &SkeletalMeshLocation(
    Inherit Velocity=true
    MaskSpawnGroup=false
    Mesh Sampling Type=::NewEnumerator5
    Module.Override Randomness=false
    Sampled Position Offset Coordinate Space:ENiagaraMeshTransforms=::NewEnumerator3
    Skeletal Mesh=DataInterface() {propertyValues {
      DefaultMesh=null
      ExcludeBoneName=root
      OptionalComponentClass=null
      PreviewMesh=/BRCosmetics/Characters/Player/Female/Medium/Bodies/F_MED_Crush/Meshes/F_MED_Crush
      SkinningMode=SkinOnTheFly
      SoftSourceActor=null
      SourceMode=Default
      UvSetIndex=0
      WholeMeshLOD=-1
      bExcludeBone=false
      bReadDeformedGeometry=true
      bRequireCurrentFrameData=true
      ComponentTags []
      FilteredBones []
      FilteredSockets [FX_0,FX_2,FX_4,FX_7]
      MeshUserParameter {parameter {name=None, typeDefHandle {registeredTypeIndex=1}, varData []}}
      SamplingRegions []
     }}
    VelocitySpeedLimit=false
   )
   /SetVariables(Particles.RibbonLinkOrder:NiagaraFloat=&MakeFloatFromInt(Int=/Niagara/DynamicInputs/DataInterface/GenerateIntFromCounter(SimpleCounter:NiagaraDataInterfaceSimpleCounter=DataInterface() {propertyValues {GpuSyncMode=None, InitialValue=0}}))) {isSetParameters=true}
  ]
  particleUpdateScript [
   &ParticleState()
   &SkeletalMeshLocation(
    Apply To Particle Position:NiagaraBool=&SetBoolByFloatComparison(A=$Particles.NormalizedAge, B=0.2, Comparison Type=::NewEnumerator6)
    Bone / Socket Index=&SpawnGroup()
    Inherit Velocity=false
    Interpolate Mesh Sampling=false
    MaskSpawnGroup=false
    Mesh Sampling Type=::NewEnumerator5
    Sampled Position Offset Coordinate Space:ENiagaraMeshTransforms=::NewEnumerator3
    Skeletal Mesh=$User.SkeletalMesh
    Socket Sampling Mode=::NewEnumerator5
   ) {module=SkeletalMeshLocation001}
   /SetVariables(
    Particles.DistanceAbove:Vector3f=&self:ClampVector(
     Input Vector:Vector3f=&Multiply_Vector(
      A=<0 0 0.1>
      B=&Subtract_Vector(A=Linked() {linkedVariable {name=Particles.Position, type=NiagaraPosition}}, B=$Output.SkeletalMeshLocation001.SampledPosition)
     )
     Max:Vector3f=<0 0 2>
     Min:Vector3f=<0 0 0>
    )
    Particles.DistanceBelow:Vector3f=&self:ClampVector(
     Input Vector:Vector3f=&Multiply_Vector(
      A=<0 0 0.1>
      B=&Subtract_Vector(A=$Output.SkeletalMeshLocation001.SampledPosition, B=Linked() {linkedVariable {name=Particles.Position, type=NiagaraPosition}})
     )
     Max:Vector3f=<0 0 1>
     Min:Vector3f=<0 0 0>
    )
    Particles.DistanceFromBone:NiagaraFloat=&Multiply_Float(
     A=0.01
     B=&VectorLength(Vector=&Subtract_Vector(A=Linked() {linkedVariable {name=Particles.Position, type=NiagaraPosition}}, B=$Output.SkeletalMeshLocation001.SampledPosition))
    )
   ) {isSetParameters=true}
   /SetVariables(Particles.NormalizedAge:NiagaraFloat=&Lerp_Float(
     A=$Particles.NormalizedAge
     B=1
     Alpha=&ClampFloat(Float=&Add_Float(A=&MakeFloatFromVector(Vector=$Particles.DistanceAbove, Channel=::NewEnumerator2), B=-1.5))
    )) {isSetParameters=true}
   &GravityForce(Gravity=&Lerp_Vector(
     A=<0 0 700>
     B=<0 0 200>
     Alpha=&ClampFloat(Float=&Multiply_Float(A=&VectorLength(Vector=$Output.SkeletalMeshLocation001.SampledVelocity), B=0.5))
    ))
   /Niagara/Modules/Update/Forces/CurlNoiseForce(
    Noise Frequency=5
    Noise Strength=&FloatFromCurve(FloatCurve=DataInterface() {propertyValues {Curve {keys [{arriveTangent=957.2371, interpMode=RCIM_Cubic, leaveTangent=957.237, tangentMode=RCTM_User, time=0, value=10}, <at 1 300>]}}})
    Pan Noise=true
    Pan Noise Field=<0 0 2>
    Random Seed:NiagaraInt32=&RandomRangeInt(Minimum=0, Maximum=100)
   )
   &Drag(
    Drag=&Multiply_Float(
     A=&ClampFloat(Float=/Niagara/DynamicInputs/TypeConversions/VectorLengthSafe(Vector:Vector3f=$Particles.Velocity), Max=5000)
     B=&Lerp_Float(A=0.5, B=0, Alpha=$Particles.NormalizedAge)
    )
    Use Linear Drag=true
    Use Rotational Drag=false
   )
   &AddVelocity(Velocity=&Multiply_VectorByFloat(Vector=$Particles.DistanceBelow, Float=&RandomRangeFloat(Minimum=0, Maximum=100))) {module=AddVelocity003}
   &AddVelocity(
    Coordinate Space=::World
    Velocity=&Lerp_Vector(
     A=<0 0 0>
     B=<0 0 100>
     Alpha=&FloatFromCurve(
      FloatCurve=DataInterface() {propertyValues {Curve {keys [{interpMode=RCIM_Cubic, tangentMode=RCTM_User, time=0.2, value=0}, <at 1 1>]}}}
      CurveIndex=$Particles.DistanceFromBone
      Scale Curve=&RandomRangeFloat(Minimum=0.25, Maximum=1)
     )
    )
   ) {enabled=false}
   /Game/Effects/Niagara/Modules/Materials/FN_RibbonMaterial/FN_RibbonMaterial(
    Color:LinearColor=&ColorFromCurve(ColorCurve=DataInterface() {propertyValues {
       AlphaCurve {keys [<at 0 1 RCIM_Cubic>]}
       BlueCurve {keys [<at 0 0.4902 RCIM_Cubic>,<at 0.3146 0.918 RCIM_Cubic>]}
       GreenCurve {keys [<at 0 1.3333 RCIM_Cubic>,<at 0.3146 0.8632 RCIM_Cubic>]}
       RedCurve {keys [<at 0 5 RCIM_Cubic>,<at 0.3146 0.8528 RCIM_Cubic>]}
      }})
    Color Pan U Speed=-2.5345
    Color Scale U=7.5
    Erosion Amount:NiagaraFloat=&Lerp_Float(A=0.5, B=1.6, Alpha=$Particles.NormalizedAge)
    Erosion Pan Speed U=-0.8346
    Erossion Scale U=4.5
    Macro Distortion Amount=1
    Opacity Pan U Speed=-1.6786
    Opacity Power:NiagaraFloat=&RandomRangeFloat(Minimum=0.3, Maximum=1)
   ) {enabled=false, module=FN_RibbonMaterial002}
   /Game/Effects/Niagara/Modules/Materials/FN_RibbonMaterial/FN_RibbonMaterial_v2(
    Color:LinearColor=&ColorFromCurve(ColorCurve=DataInterface() {propertyValues {
       AlphaCurve {keys [<at 0 1 RCIM_Cubic>]}
       BlueCurve {keys [<at 0 0.4902 RCIM_Cubic>,<at 0.3146 0.918 RCIM_Cubic>]}
       GreenCurve {keys [<at 0 1.3333 RCIM_Cubic>,<at 0.3146 0.8632 RCIM_Cubic>]}
       RedCurve {keys [<at 0 5 RCIM_Cubic>,<at 0.3146 0.8528 RCIM_Cubic>]}
      }})
    Color Pan U Speed=-2.5345
    Color Scale U=7.5
    Erosion Amount:NiagaraFloat=&Lerp_Float(A=0.5, B=1.6, Alpha=$Particles.NormalizedAge)
    Erosion Pan Speed U=-0.8346
    Erossion Scale U=4.5
    Macro Distortion Amount=1
    Opacity Pan U Speed=-1.6786
    Opacity Power:NiagaraFloat=&RandomRangeFloat(Minimum=0.3, Maximum=1)
   ) {enabled=false}
   /Game/Effects/Niagara/Modules/Materials/FN_RibbonMaterial/FN_Ribbon_v3/FN_RibbonMaterial(
    Color=&ColorFromCurve(ColorCurve=DataInterface() {propertyValues {
       AlphaCurve {keys [<at 0 2 RCIM_Cubic>]}
       BlueCurve {keys [<at 0 1.3852 RCIM_Cubic>,<at 0.3146 0.918 RCIM_Cubic>]}
       GreenCurve {keys [<at 0 3.5667 RCIM_Cubic>,<at 0.3146 0.8632 RCIM_Cubic>]}
       RedCurve {keys [<at 0 5 RCIM_Cubic>,<at 0.3146 0.8528 RCIM_Cubic>]}
      }})
    Color Pan U Speed=-2.5345
    Color Scale U=7.5
    Erosion Amount:NiagaraFloat=&Lerp_Float(A=0.5, B=1.6, Alpha=$Particles.NormalizedAge)
    Erosion Pan Speed U=-0.8346
    Erossion Scale U=4.5
    Macro Distortion Amount=1
    Opacity Pan U Speed=-1.6786
    Opacity Power:NiagaraFloat=&RandomRangeFloat(Minimum=0.3, Maximum=1)
    Scale Alpha:NiagaraFloat=&FloatFromCurve(
     FloatCurve=DataInterface() {propertyValues {Curve {keys [<at 0 0>,<at 1 1>]}}}
     CurveIndex=&Divide_Float(A=&VectorLength(Vector=$Engine.Owner.Velocity), B=500)
    )
   ) {module=FN_RibbonMaterial001}
   &ScaleRibbonWidth(Ribbon Width Scale=&FloatFromCurve(
     FloatCurve=DataInterface() {propertyValues {Curve {keys [
         {arriveTangent=4.0699, interpMode=RCIM_Cubic, leaveTangent=4.0699, tangentMode=RCTM_User, time=0, value=0}
         <at 0.5 0.9666 RCIM_Cubic>
         {arriveTangent=-2.5115, interpMode=RCIM_Cubic, leaveTangent=-2.5115, tangentMode=RCTM_User, time=1, value=0}
        ]}}}
     Scale Curve=&RandomRangeFloat(Minimum=0.5, Maximum=2)
    )) {module=ScaleRibbonWidth001}
   &SolveForcesAndVelocity(Clamp Velocity=false, Limit Acceleration=false)
   &Color(Color=&ColorFromCurve(
     ColorCurve=DataInterface() {propertyValues {
       AlphaCurve {keys [<at 0 1 RCIM_Cubic>]}
       BlueCurve {keys [<at 0 0 RCIM_Cubic>,<at 0.4 0>,<at 0.6 0>,<at 1 1 RCIM_Cubic>]}
       GreenCurve {keys [<at 0 0.0148 RCIM_Cubic>,<at 0.4 0.8005>,<at 0.6 1>,<at 1 0.087 RCIM_Cubic>]}
       RedCurve {keys [<at 0 1 RCIM_Cubic>,<at 0.4 1>,<at 0.6 0.0681>,<at 1 0 RCIM_Cubic>]}
      }}
     CurveIndex=$Particles.DistanceFromBone
    )) {enabled=false}
  ]
  rendererClasses [&NiagaraSpriteRendererProperties, &NiagaraRibbonRendererProperties]
  renderers [
   {rendererClass=&NiagaraSpriteRendererProperties}
   {rendererClass=&NiagaraRibbonRendererProperties, rendererIndex=1}
  ]
 }
}
summary {
 systemName=NS_Backpack_Crush_Fuse
 emitters [
  {
   emitterName=FX_Fuse_L_Hand001
   rendererClasses [
    &NiagaraLightRendererProperties
    &NiagaraMeshRendererProperties
    &NiagaraSpriteRendererProperties
   ]
  }
  {emitterName=Smoke004, rendererClasses [&NiagaraSpriteRendererProperties, &NiagaraRibbonRendererProperties]}
 ]
 userVariables [User.SkeletalMesh {
   type=NiagaraDataInterfaceSkeletalMesh
   defaultValue {
    struct=NiagaraExt_VariableValue_DataInterface
    value {dataInterface=&self:NiagaraDataInterfaceSkeletalMesh_0}
   }
  }]
}
