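# Parent masked material for the firecracker on the Crush backpack cosmetic, animating its visibility with a timed reveal and hide.
# Standard character PreFX, CharacterSkinFX and PostFX functions build the material attributes, then the opacity mask is overridden.
# The mask is a SphereMask on normalized bounding-box X around X_Center, its radius driven by timestamped linear in/out transitions.
# Dithered masking softens the moving edge.
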Material /BRCosmetics/Accessories/FORT_Backpacks/Backpack_F_MED_Crush/Materials/F_MED_Crush_Backpack_Firecracker_Parent
graph {
 properties {
  bUseMaterialAttributes=true
  bUsedWithMeshDeformer=true
  bUsedWithSkeletalMesh=true
  blendMode=BLEND_Masked
  ditherOpacityMask=true
  opacityMaskClipValue=0.5
  pixelDepthOffsetMode=PDOM_Legacy
  refractionMethod=RM_IndexOfRefraction
  usedShadingModels=Default Lit
 }
 nodes [
  MaterialFunctionCall!(Time, Divisor=Constant() {r=2}, Mod) {materialFunction=/Game/Effects/Fort_Effects/Materials/_Functions/MF_ModuloTime}
  n1 = MaterialFunctionCall() {materialFunction=&BaseFN_CharacterFunction_PreFX}
  n2 = MaterialFunctionCall(
   AO=@n1.AO
   Distance Blend=@n1.AO
   Fresnel=@n1.AO
   Gmap=@n1.AO
   Pre Skinned Local Position=@n1.AO
   Material=MaterialFunctionCall(Gmap=@n1.Gmap, MaterialAttributes=@n1.Gmap, AO=@n1.Gmap) {materialFunction=&CharacterSkinFX}.Result
  ) {materialFunction=&BaseFN_CharacterFunction_PostFX}
  n3 = ScalarParameter() {parameterName=TimestampA}
  n4 = ScalarParameter() {defaultValue=0.4, parameterName=TransitionLength_In}
  n5 = ScalarParameter() {defaultValue=10000000000, parameterName=TimestampB}
  n6 = ScalarParameter() {defaultValue=0.2, parameterName=TransitionLength_Out}
  n7 = StaticBool()
  SetMaterialAttributes!(
   MaterialAttributes=@n2.Result
   Opacity Mask=SphereMask(
    A=MaterialFunctionCall() {materialFunction=&BoundingBoxBased_0-1_UVW}.R
    B=ScalarParameter() {defaultValue=0.71, parameterName=X_Center}
    Radius=MaterialFunctionCall(TimestampA=@n3, TransitionLength_In=@n4, TimestampB=@n5, TransitionLength_Out=@n6, Transition In Only?=@n7) {materialFunction=&MF_Timestamp_InOut_01}.Linear
    Hardness=Constant() {r=0.65}
   )
  ) {attributeSetTypes [#1]}
 ]
}
