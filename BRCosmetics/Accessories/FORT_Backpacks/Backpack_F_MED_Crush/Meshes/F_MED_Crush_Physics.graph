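# Physics asset for the F_MED_Crush backpack cosmetic, defining simple collision bodies and joints for its jaw and fuse bones.
# A box body covers the pack, a capsule covers the jaw, and small capsules cover each segment of the fuse.
# The jaw is tightly constrained to the pack, and the fuse forms a chain off the jaw with looser, uniform swing and twist limits at each joint.
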PhysicsAsset /BRCosmetics/Accessories/FORT_Backpacks/Backpack_F_MED_Crush/Meshes/F_MED_Crush_Physics
bodies {
 dyn_fuse_01 [{shapeType=Capsule, length=0.505, radius=0.505, center=<0.8155 0 0>, rotation=<rot 90 0 0>}]
 dyn_fuse_02 [{shapeType=Capsule, length=0.505, radius=0.505, center=<0.6839 -0.0083 0.0442>, rotation=<rot 0 89.3383 -88.8871>}]
 dyn_fuse_03 [{shapeType=Capsule, length=0.9, radius=0.505, center=<0.8999 0 0>, rotation=<rot 90 0 0>}]
 dyn_fuse_04 [{shapeType=Capsule, length=0.8833, radius=0.505, center=<0.9278 0 0>, rotation=<rot 90 0 0>}]
 dyn_fuse_05 [{shapeType=Capsule, length=0.505, radius=0.505, center=<0.5364 0 0>, rotation=<rot 90 0 0>}]
 dyn_jaw [{shapeType=Capsule, length=7.8115, radius=1.8031, rotation=<rot 0 0 90>}]
 pack [{shapeType=Box, extentX=11.8677, extentY=12.4942, extentZ=17.4168, center=<0 -3.8912 -2.4068>}]
}
constraints [
 {bone1Name=dyn_fuse_01, bone2Name=dyn_jaw, swing1LimitDegrees=25, swing2LimitDegrees=25, twistLimitDegrees=25}
 {bone1Name=dyn_fuse_02, bone2Name=dyn_fuse_01, swing1LimitDegrees=25, swing2LimitDegrees=25, twistLimitDegrees=25}
 {bone1Name=dyn_fuse_03, bone2Name=dyn_fuse_02, swing1LimitDegrees=25, swing2LimitDegrees=25, twistLimitDegrees=25}
 {bone1Name=dyn_fuse_04, bone2Name=dyn_fuse_03, swing1LimitDegrees=25, swing2LimitDegrees=25, twistLimitDegrees=25}
 {bone1Name=dyn_fuse_05, bone2Name=dyn_fuse_04, swing1LimitDegrees=25, swing2LimitDegrees=25, twistLimitDegrees=25}
 {bone1Name=dyn_jaw, bone2Name=pack, swing1LimitDegrees=2.5, swing2LimitDegrees=2.5, twistLimitDegrees=2.5}
]
