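# Physics asset for the wings of the F_MED_Crush backpack cosmetic, providing collision bodies and joints for the two dynamic wing bones.
# Each wing bone carries a capsule, and the pack bone carries a box body.
# Each wing is constrained directly to the pack, with non-zero limit angles only on the second swing axis, suggesting a hinge-like flap.
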PhysicsAsset /BRCosmetics/Accessories/FORT_Backpacks/Backpack_F_MED_Crush/Meshes/F_MED_Crush_Wings_Physics
bodies {
 dyn_l_wing [{shapeType=Capsule, length=10.0358, radius=3.4806, center=<4.8115 0.0001 1.9013>, rotation=<rot 0 0 90>}]
 dyn_r_wing [{shapeType=Capsule, length=10.1559, radius=3.3606, center=<4.6507 0 1.3587>, rotation=<rot 0 0 90>}]
 pack [{shapeType=Box, extentX=11.8677, extentY=12.4942, extentZ=17.4168, center=<0 -3.8912 -2.4068>}]
}
constraints [
 {bone1Name=dyn_r_wing, bone2Name=pack, swing1LimitDegrees=0, swing2LimitDegrees=45, twistLimitDegrees=0}
 {bone1Name=dyn_l_wing, bone2Name=pack, swing1LimitDegrees=0, swing2LimitDegrees=45, twistLimitDegrees=0}
]
