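# Animation Blueprint for the F_MED_Crush backpack cosmetic, adding physics motion to its wings, body and jaw bone.
# Two wing RigidBody sims of differing strength blend on jumping, falling or crouching, then feed a body RigidBody sim and a slight additive dyn_jaw offset.
# On customization load, the EventGraph watches the owning pawn's AthenaKills stat and, after a delay, flags kill-count increases.
# The AnimGraph shown does not read those kill flags.
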AnimBlueprint /BRCosmetics/Accessories/FORT_Backpacks/Backpack_F_MED_Crush/Meshes/F_MED_Crush_AnimBP
target_skeleton=&/F_MED_Crush_Skeleton
graphs [
 AnimGraph {nodes [
   Local Space Ref Pose()
   n1 = RigidBody(ComponentPose=Local To Component(LocalPose=Input Pose() {name=InPose_0}), Alpha=0.25) {
    bClampLinearTranslationLimitToRefPose=true
    bOverrideWorldGravity=true
    overridePhysicsAsset=&/F_MED_Crush_Wings_Physics
    simulationSpace=BaseBoneSpace
    alphaBoolBlend {blendOption=Linear}
    baseBoneRef {boneName=root}
    componentLinearAccScale=<0.5 0.5 0.5>
    overrideWorldGravity {z=-980}
   }
   n2 = Component To Local(ComponentPose=@n1)
   n3 = RigidBody(ComponentPose=Local To Component(LocalPose=Input Pose()), Alpha=0.4) {
    bClampLinearTranslationLimitToRefPose=true
    bOverrideWorldGravity=true
    overridePhysicsAsset=&F_MED_Crush_Wings_Physics
    simulationSpace=BaseBoneSpace
    alphaBoolBlend {blendOption=Linear}
    baseBoneRef {boneName=root}
    componentLinearAccScale=<0.5 0.5 0.5>
    overrideWorldGravity {z=-980}
   }
   n4 = Component To Local(ComponentPose=@n3)
   n5 = OR Boolean!(A=Get bIsJumping(), B=Get bIsFalling(), C:bool=Get bIsCrouching()) -> (ReturnValue)
   n6 = RigidBody(
    ComponentPose=Local To Component(LocalPose=Blend Poses by bool(BlendPose_0=@n2, BlendPose_1=@n4, bActiveValue=@n5) {blendType=Linear})
    Alpha=0.4
   ) {
    bClampLinearTranslationLimitToRefPose=true
    bOverrideWorldGravity=true
    overridePhysicsAsset=&/F_MED_Crush_Physics
    simulationSpace=BaseBoneSpace
    alphaBoolBlend {blendOption=Linear}
    baseBoneRef {boneName=root}
    componentAppliedLinearAccClamp {z=1000}
    componentLinearAccScale=<0.4 0.4 0.4>
    overrideWorldGravity {z=-980}
   }
   n7 = "Transform (Modify) Bone - Bone: dyn_jaw"(ComponentPose=@n6, Alpha=1, Translation="0,0,0.2", Rotation="0,0,0", Scale="1,1,1") {translationMode=BMM_Additive, alphaBoolBlend {blendOption=Linear}, boneToModify {boneName=dyn_jaw}, translation {z=0.2}}
   Output Pose(Result=Component To Local(ComponentPose=@n7)) {layerGroup=None}
  ]}
 EventGraph {nodes [
   Event OnCharacterCustomizationLoaded()
   > n1 = Cast To FortPlayerPawn(Object=TryGetFortPawn())
   > n2 = WatchCosmeticStat!(
    execute:Exec
    Pawn:Fort Player Pawn Object Reference=@n1
    StatTag:Fort Stat Manager Tag Structure=(TagName="StatManager.AthenaKills")
    DefaultValue:int=0
    bBroadcastInitialValue:bool=true
    InitialValueBroadcastDelay:float=-1
    ChangeValueBroadcastDelay:float=-1
    bConsumeWaitingNotifications:bool=true
   ) -> (then:Exec, OnStatChanged:Exec, StatName:Name, StatValue:int) {is=title K2Node_AsyncAction}
   Delay(execute=@n2.OnStatChanged, Duration=1.5)
   > Set Did_Kill(Did_Kill:bool=false) -> (Output_Get:bool)
   > n3 = Set KillCount(KillCount:double="To Float (Integer)"(InInt=@n2.StatValue)) -> (Output_Get:double)
   > Branch(Condition="float > float"(A=@n3, B=Get Last_KillCount() -> (Last_KillCount:double)) {is=title K2Node_CallFunction})
   > Set Active_Kill(Active_Kill:bool=true) -> (Output_Get:bool)
   > Set Last_KillCount(Last_KillCount:double=@n3) -> (Output_Get:double)
   Event BlueprintInitializeAnimation()
   > Set Skel_Mesh(Skel_Mesh=GetOwnerSkeletalMeshComponent())
  ]}
]
variables [
 Active_Kill {type=boolean}
 Did_Kill {type=boolean}
 KillCount {type=number}
 Last_KillCount {type=number}
 Skel_Mesh {type=SkeletalMeshComponent}
 Surprise_Alpha {type=number}
 Surprise_Length_Out {type=number}
 Surprise_Timer {type=number}
 Timer {type=boolean}
]
